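# Translucent unlit material for Niagara mesh particles, named as the energy sphere effect on the Alien Trooper backpack.
# Emissive layers Fresnel rims with panning smoke, rounded noise bands and electric noise, tinted by particle color.
# A Use_Beam_mesh_WPO scalar switches to flat particle color and noise-driven vertex displacement, masked by a sine along V and driven by dynamic parameters for speed and intensity.
# A UseParticleAlpha scalar blends in particle alpha.
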Material /BRCosmetics/Accessories/FORT_Backpacks/Backpack_M_MED_Alien_Trooper/FX/M_Backpack_Alien_Trooper_Sphere
graph {
 properties {
  bUsedWithNiagaraMeshParticles=true
  blendMode=BLEND_Translucent
  pixelDepthOffsetMode=PDOM_Legacy
  shadingModel=MSM_Unlit
  usedShadingModels=Unlit
 }
 nodes [
  n1 = TextureCoordinate() {uTiling=0, vTiling=1.5}
  n2 = VertexColor()
  n3 = Multiply(A=Time(), B=DynamicParameter() {paramNames [Panner_Speed, WPO_Intensity, Param3, Param4]}.Panner_Speed)
  n4 = Panner(Coordinate=Add(A=Multiply(A=@n1, B=@n2.R), B=@n2.R), Time=@n3) {bFractionalPart=true, speedX=0.1, speedY=0.8}
  n5 = Panner(Coordinate=Add(A=TextureCoordinate() {uTiling=0, vTiling=1.2}, B=VertexColor().R), Time=@n3) {bFractionalPart=true, speedY=0.8}
  n6 = LinearInterpolate(
   A=Power(Base=TextureSample(UVs=@n4) {samplerType=SAMPLERTYPE_Grayscale, texture=&LowResBlurredNoise}.R)
   B=LinearInterpolate(Alpha=TextureSample(UVs=@n5) {samplerType=SAMPLERTYPE_Grayscale, texture=&LowResBlurredNoise}.R) {constA=-1}
  )
  n7 = Multiply(
   A=Multiply(
    A=LinearInterpolate(Alpha=@n6) {constA=-0.5, constB=2.5}
    B=DynamicParameter() {paramNames [Panner_Speed, WPO_Intensity, Param3, Param4]}.WPO_Intensity
   ) {constB=0.8}
   B=Constant3Vector() {constant=<rgba 1 1.2 1 0>}
  )
  n8 = TextureCoordinate()
  n9 = Multiply(A=Sine(None=ComponentMask(None=@n8) {g=true}) {period=2}) {constB=2}
  LinearInterpolate(
   A=Constant()
   B=Multiply(A=Transform(None=@n7) {transformSourceType=TRANSFORMSOURCE_Local}, B=Saturate(None=@n9))
   Alpha=ScalarParameter() {parameterName=Use_Beam_mesh_WPO}
  )
  n10 = ParticleColor()
  LinearInterpolate(A=Constant() {r=1}, B=@n10.A, Alpha=ScalarParameter() {parameterName=UseParticleAlpha})
  n11 = ParticleColor()
  n12 = Fresnel() {baseReflectFraction=0, exponent=4}
  n13 = Panner(Coordinate=TextureCoordinate()) {speedX=0.1, speedY=0.1}
  n14 = Add(
   A=Panner(Coordinate=TextureCoordinate() {vTiling=0.5}) {speedY=0.7}
   B=Multiply(A=TextureSample(UVs=@n13) {texture=/Game/Effects/Fort_Effects/Textures/Tile/T_Inky_Smoke_Tile2}.R) {constB=0.4}
  )
  n15 = LinearInterpolate(Alpha=Power(Base=TextureSample(UVs=@n14) {texture=&T_Inky_Smoke_Tile}.B) {constExponent=3}) {constA=-0.1, constB=2}
  n16 = ComponentMask(None=TextureCoordinate()) {g=true}
  n17 = TextureSample(UVs=Panner(Coordinate=TextureCoordinate() {uTiling=2, vTiling=0.5}) {speedY=0.3}) {samplerType=SAMPLERTYPE_Grayscale, texture=&LowResBlurredNoise}
  n18 = Panner(Coordinate=TextureCoordinate() {vTiling=2}) {speedY=1}
  n19 = Add(
   A=LinearInterpolate(Alpha=Power(Base=@n17.R)) {constA=-0.32}
   B=Power(Base=TextureSample(UVs=@n18) {samplerType=SAMPLERTYPE_Grayscale, texture=&LowResBlurredNoise}.R)
  )
  n20 = Fresnel() {baseReflectFraction=0, exponent=3}
  n21 = ScalarParameter() {defaultValue=0.15, parameterName=FresnelPower}
  n22 = Multiply(A=Multiply(A=Add(A=@n16, B=@n19), B=Power(Base=@n20, Exp=@n21) {constExponent=0.15})) {constB=4}
  n23 = Saturate(None=Saturate(None=Round(None=@n22)))
  n24 = Constant() {r=0.01}
  n25 = Power(Base=Power(Base=Fresnel() {baseReflectFraction=0, exponent=4}) {constExponent=0.8}) {constExponent=3}
  n26 = Power(Base=Fresnel() {baseReflectFraction=0, exponent=4}) {constExponent=1}
  n27 = TextureCoordinate() {vTiling=0.8}
  n28 = Power(Base=TextureSample(UVs=Panner(Coordinate=@n27) {speedY=0.5}) {samplerType=SAMPLERTYPE_LinearColor, texture=/Game/Characters/Player/Male/Medium/Bodies/M_MED_SRDrift/Textures/T_ElectricNoise}.R)
  n29 = Multiply(A=Round(None=Multiply(A=@n28) {constB=30})) {constB=10}
  n30 = Fresnel() {baseReflectFraction=0, exponent=1}
  n31 = Add(
   A=Add(A=Multiply(A=@n25) {constB=15}, B=Multiply(A=@n26, B=@n29))
   B=LinearInterpolate(Alpha=Power(Base=@n30) {constExponent=3}) {constA=0.05}
  ) {constB=0.05}
  n32 = Add(
   A=Multiply(A=Power(Base=@n12) {constExponent=0.2}, B=Multiply(A=@n15))
   B=Add(A=Multiply(A=@n23) {constB=10}, B=LinearInterpolate(A=@n24, Alpha=@n31))
  )
  MaterialFunctionCall(Emissive=LinearInterpolate(A=Multiply(A=@n11.RGB, B=@n32), B=@n10.RGB, Alpha=ScalarParameter() {parameterName=Use_Beam_mesh_WPO})) {materialFunction=&MF_FX_EmissiveAdjustments}
  VectorParameter() {defaultValue=<rgba 1 0 0.667 1>}
 ]
}
