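# Physics asset for the Alien Trooper backpack cosmetic's skeletal mesh, covering the master pack bone and its floating pieces.
# Small capsule bodies sit on Dyn_Master_Pack and on three left and three right float-piece bones.
# Each piece is constrained directly to the master body in a star layout rather than chained.
# Swing and twist are locked on every constraint, so the stepped per-piece angle limits are inactive.
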PhysicsAsset /BRCosmetics/Accessories/FORT_Backpacks/Backpack_M_MED_Alien_Trooper/Meshes/M_MED_Alien_Trooper_Pack_Physics
bodies {
 Dyn_Left___Float_Pieces1 [{shapeType=Capsule, length=2, radius=1, center=<0.5 0 0>, rotation=<rot 90 0 0>}]
 Dyn_Left___Float_Pieces2 [{shapeType=Capsule, length=2, radius=1, center=<0.5 0 0>, rotation=<rot 90 0 0>}]
 Dyn_Left___Float_Pieces3 [{shapeType=Capsule, length=2, radius=1, center=<0.5 0 0>, rotation=<rot 90 0 0>}]
 Dyn_Master_Pack [{shapeType=Capsule, length=0.505, radius=0.505}]
 Dyn_Right_Float_Pieces1 [{shapeType=Capsule, length=2, radius=1, center=<0.5 0 0>, rotation=<rot 90 0 0>}]
 Dyn_Right_Float_Pieces2 [{shapeType=Capsule, length=2, radius=1, center=<0.5 0 0>, rotation=<rot 90 0 0>}]
 Dyn_Right_Float_Pieces3 [{shapeType=Capsule, length=2, radius=1, center=<0.5 0 0>, rotation=<rot 90 0 0>}]
}
constraints [
 {
  bone1Name=Dyn_Right_Float_Pieces1
  bone2Name=Dyn_Master_Pack
  swing1LimitDegrees=30
  swing1Motion=Locked
  swing2LimitDegrees=30
  swing2Motion=Locked
  twistLimitDegrees=30
  twistMotion=Locked
 }
 {
  bone1Name=Dyn_Right_Float_Pieces2
  bone2Name=Dyn_Master_Pack
  swing1LimitDegrees=10
  swing1Motion=Locked
  swing2LimitDegrees=10
  swing2Motion=Locked
  twistLimitDegrees=10
  twistMotion=Locked
 }
 {
  bone1Name=Dyn_Right_Float_Pieces3
  bone2Name=Dyn_Master_Pack
  swing1LimitDegrees=5
  swing1Motion=Locked
  swing2LimitDegrees=5
  swing2Motion=Locked
  twistLimitDegrees=5
  twistMotion=Locked
 }
 {
  bone1Name=Dyn_Left___Float_Pieces1
  bone2Name=Dyn_Master_Pack
  swing1LimitDegrees=30
  swing1Motion=Locked
  swing2LimitDegrees=30
  swing2Motion=Locked
  twistLimitDegrees=30
  twistMotion=Locked
 }
 {
  bone1Name=Dyn_Left___Float_Pieces2
  bone2Name=Dyn_Master_Pack
  swing1LimitDegrees=10
  swing1Motion=Locked
  swing2LimitDegrees=10
  swing2Motion=Locked
  twistLimitDegrees=10
  twistMotion=Locked
 }
 {
  bone1Name=Dyn_Left___Float_Pieces3
  bone2Name=Dyn_Master_Pack
  swing1LimitDegrees=5
  swing1Motion=Locked
  swing2LimitDegrees=5
  swing2Motion=Locked
  twistLimitDegrees=5
  twistMotion=Locked
 }
]
