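# Animation Blueprint for the Alien Trooper backpack that makes six left and right floating pieces bob and twist.
# The output chain runs rigid-body physics in root bone space, then additive bone offsets, some enabled only in the front end, then ambient translation and rotation.
# Each update, Ring_Animation derives those ambient values from game-time sine waves.
# A ring-pose blend chain and its delay sequence are left disconnected.
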AnimBlueprint /BRCosmetics/Accessories/FORT_Backpacks/Backpack_M_MED_Alien_Trooper/Meshes/M_MED_Alien_Trooper_Pack_AnimBP
target_skeleton=&/M_MED_Alien_Trooper_Pack_Skeleton
graphs [
 AnimGraph {nodes [
   Local Space Ref Pose()
   Local To Component()
   n1 = "Transform (Modify) Bone - Bone: Dyn_Right_Float_Pieces1"(Alpha=1, Translation="1.5,0,0", Rotation="0,0,0", Scale="1,1,1") {
    rotationMode=BMM_Additive
    translationMode=BMM_Additive
    translationSpace=BCS_BoneSpace
    alphaBoolBlend {blendOption=Linear}
    boneToModify {boneName=Dyn_Right_Float_Pieces1}
    translation {x=1.5}
   }
   "Transform (Modify) Bone - Bone: Dyn_Right_Float_Pieces2"(ComponentPose=@n1, Alpha=1, Translation="1.5,0,0", Rotation="0,0,0", Scale="1,1,1") {
    rotationMode=BMM_Additive
    translationMode=BMM_Additive
    translationSpace=BCS_BoneSpace
    alphaBoolBlend {blendOption=Linear}
    boneToModify {boneName=Dyn_Right_Float_Pieces2}
    translation {x=1.5}
   }
   n2 = "Transform (Modify) Bone - Bone: Dyn_Left___Float_Pieces1"(Alpha=1, Translation="1.5,0,0", Rotation="0,0,0", Scale="1,1,1") {
    rotationMode=BMM_Additive
    translationMode=BMM_Additive
    translationSpace=BCS_BoneSpace
    alphaBoolBlend {blendOption=Linear}
    boneToModify {boneName=Dyn_Left___Float_Pieces1}
    translation {x=1.5}
   }
   "Transform (Modify) Bone - Bone: Dyn_Left___Float_Pieces2"(ComponentPose=@n2, Alpha=1, Translation="1.5,0,0", Rotation="0,0,0", Scale="1,1,1") {
    rotationMode=BMM_Additive
    translationMode=BMM_Additive
    translationSpace=BCS_BoneSpace
    alphaBoolBlend {blendOption=Linear}
    boneToModify {boneName=Dyn_Left___Float_Pieces2}
    translation {x=1.5}
   }
   n3 = RigidBody(ComponentPose=Local To Component(LocalPose=Input Pose())) {
    bClampLinearTranslationLimitToRefPose=true
    bOverrideWorldGravity=true
    overridePhysicsAsset=&/M_MED_Alien_Trooper_Pack_Physics
    simulationSpace=BaseBoneSpace
    alphaBoolBlend {blendOption=Linear}
    baseBoneRef {boneName=root}
    componentLinearAccScale=<0.1 0.5 0.2>
    componentLinearVelScale=<0.1 0.5 0.2>
    overrideWorldGravity {z=-980}
   }
   n4 = "Transform (Modify) Bone - Bone: Dyn_Right_Float_Pieces3"(ComponentPose=@n3, Alpha=1, Translation="1.5,0,0", Rotation="0,0,0", Scale="1,1,1") {
    rotationMode=BMM_Additive
    translationMode=BMM_Additive
    translationSpace=BCS_BoneSpace
    alphaBoolBlend {blendOption=Linear}
    boneToModify {boneName=Dyn_Right_Float_Pieces3}
    translation {x=1.5}
   }
   n5 = "Transform (Modify) Bone - Bone: Dyn_Left___Float_Pieces3"(ComponentPose=@n4, Alpha=1, Translation="1.5,0,0", Rotation="0,0,0", Scale="1,1,1") {
    rotationMode=BMM_Additive
    translationMode=BMM_Additive
    translationSpace=BCS_BoneSpace
    alphaBoolBlend {blendOption=Linear}
    boneToModify {boneName=Dyn_Left___Float_Pieces3}
    translation {x=1.5}
   }
   n6 = Get bIsInFrontEnd()
   n7 = "Transform (Modify) Bone - Bone: Dyn_Right_Float_Pieces1"!(ComponentPose=@n5, bAlphaBoolEnabled:bool=@n6, Translation="-1.5,0,0", Rotation="0,0,0", Scale="1,1,1") -> (Pose) {
    alphaInputType=Bool
    rotationMode=BMM_Additive
    translationMode=BMM_Additive
    translationSpace=BCS_BoneSpace
    alphaBoolBlend {blendOption=Linear}
    boneToModify {boneName=Dyn_Right_Float_Pieces1}
    translation {x=-1.5}
   }
   n8 = "Transform (Modify) Bone - Bone: Dyn_Left___Float_Pieces1"!(ComponentPose=@n7, bAlphaBoolEnabled:bool=@n6, Translation="-1.5,0,0", Rotation="0,0,0", Scale="1,1,1") -> (Pose) {
    alphaInputType=Bool
    rotationMode=BMM_Additive
    translationMode=BMM_Additive
    translationSpace=BCS_BoneSpace
    alphaBoolBlend {blendOption=Linear}
    boneToModify {boneName=Dyn_Left___Float_Pieces1}
    translation {x=-1.5}
   }
   n9 = "Transform (Modify) Bone - Bone: Dyn_Right_Float_Pieces3"!(ComponentPose=@n8, bAlphaBoolEnabled:bool=@n6, Translation="1.2,0,0", Rotation="0,0,0", Scale="1,1,1") -> (Pose) {
    alphaInputType=Bool
    rotationMode=BMM_Additive
    translationMode=BMM_Additive
    translationSpace=BCS_BoneSpace
    alphaBoolBlend {blendOption=Linear}
    boneToModify {boneName=Dyn_Right_Float_Pieces3}
    translation {x=1.2}
   }
   n10 = "Transform (Modify) Bone - Bone: Dyn_Left___Float_Pieces3"!(ComponentPose=@n9, bAlphaBoolEnabled:bool=@n6, Translation="1.2,0,0", Rotation="0,0,0", Scale="1,1,1") -> (Pose) {
    alphaInputType=Bool
    rotationMode=BMM_Additive
    translationMode=BMM_Additive
    translationSpace=BCS_BoneSpace
    alphaBoolBlend {blendOption=Linear}
    boneToModify {boneName=Dyn_Left___Float_Pieces3}
    translation {x=1.2}
   }
   n11 = "Transform (Modify) Bone - Bone: Dyn_Right_Float_Pieces1"!(ComponentPose=@n10, Alpha:float=1, Translation=Get Ambient_Trans_R1() -> (Ambient_Trans_R1:Vector), Rotation="0,0,0", Scale="1,1,1") -> (Pose) {
    rotationMode=BMM_Additive
    translationMode=BMM_Additive
    translationSpace=BCS_BoneSpace
    alphaBoolBlend {blendOption=Linear}
    boneToModify {boneName=Dyn_Right_Float_Pieces1}
   }
   n12 = "Transform (Modify) Bone - Bone: Dyn_Right_Float_Pieces2"(ComponentPose=@n11, Alpha=1, Translation=Get Ambient_Trans_R2() -> (Ambient_Trans_R2:Vector), Rotation="0,0,0", Scale="1,1,1") {
    rotationMode=BMM_Additive
    translationMode=BMM_Additive
    translationSpace=BCS_BoneSpace
    alphaBoolBlend {blendOption=Linear}
    boneToModify {boneName=Dyn_Right_Float_Pieces2}
   }
   n13 = "Transform (Modify) Bone - Bone: Dyn_Right_Float_Pieces3"!(ComponentPose=@n12, Alpha:float=1, Translation=Get Ambient_Trans_R3() -> (Ambient_Trans_R3:Vector), Rotation="0,0,0", Scale="1,1,1") -> (Pose) {
    rotationMode=BMM_Additive
    translationMode=BMM_Additive
    translationSpace=BCS_BoneSpace
    alphaBoolBlend {blendOption=Linear}
    boneToModify {boneName=Dyn_Right_Float_Pieces3}
   }
   n14 = "Transform (Modify) Bone - Bone: Dyn_Left___Float_Pieces1"!(ComponentPose=@n13, Alpha:float=1, Translation=Get Ambient_Trans_L1() -> (Ambient_Trans_L1:Vector), Rotation="0,0,0", Scale="1,1,1") -> (Pose) {
    rotationMode=BMM_Additive
    translationMode=BMM_Additive
    translationSpace=BCS_BoneSpace
    alphaBoolBlend {blendOption=Linear}
    boneToModify {boneName=Dyn_Left___Float_Pieces1}
   }
   n15 = "Transform (Modify) Bone - Bone: Dyn_Left___Float_Pieces2"(ComponentPose=@n14, Alpha=1, Translation=Get Ambient_Trans_L2() -> (Ambient_Trans_L2:Vector), Rotation="0,0,0", Scale="1,1,1") {
    rotationMode=BMM_Additive
    translationMode=BMM_Additive
    translationSpace=BCS_BoneSpace
    alphaBoolBlend {blendOption=Linear}
    boneToModify {boneName=Dyn_Left___Float_Pieces2}
   }
   n16 = "Transform (Modify) Bone - Bone: Dyn_Left___Float_Pieces3"!(ComponentPose=@n15, Alpha:float=1, Translation=Get Ambient_Trans_L3() -> (Ambient_Trans_L3:Vector), Rotation="0,0,0", Scale="1,1,1") -> (Pose) {
    rotationMode=BMM_Additive
    translationMode=BMM_Additive
    translationSpace=BCS_BoneSpace
    alphaBoolBlend {blendOption=Linear}
    boneToModify {boneName=Dyn_Left___Float_Pieces3}
   }
   n17 = "Transform (Modify) Bone - Bone: Dyn_Right_Float_Pieces1"(ComponentPose=@n16, Alpha=1, Translation="0,0,0", Rotation=Get Ambient Rot R1() -> (Ambient Rot R1:Rotator), Scale="1,1,1") {
    rotationMode=BMM_Additive
    translationMode=BMM_Additive
    translationSpace=BCS_BoneSpace
    alphaBoolBlend {blendOption=Linear}
    boneToModify {boneName=Dyn_Right_Float_Pieces1}
   }
   n18 = "Transform (Modify) Bone - Bone: Dyn_Right_Float_Pieces2"(ComponentPose=@n17, Alpha=1, Translation="0,0,0", Rotation=Get Ambient Rot R2() -> (Ambient Rot R2:Rotator), Scale="1,1,1") {
    rotationMode=BMM_Additive
    translationMode=BMM_Additive
    translationSpace=BCS_BoneSpace
    alphaBoolBlend {blendOption=Linear}
    boneToModify {boneName=Dyn_Right_Float_Pieces2}
   }
   n19 = "Transform (Modify) Bone - Bone: Dyn_Right_Float_Pieces3"(ComponentPose=@n18, Alpha=1, Translation="0,0,0", Rotation=Get Ambient Rot R3() -> (Ambient Rot R3:Rotator), Scale="1,1,1") {
    rotationMode=BMM_Additive
    translationMode=BMM_Additive
    translationSpace=BCS_BoneSpace
    alphaBoolBlend {blendOption=Linear}
    boneToModify {boneName=Dyn_Right_Float_Pieces3}
   }
   n20 = "Transform (Modify) Bone - Bone: Dyn_Left___Float_Pieces1"(ComponentPose=@n19, Alpha=1, Translation="0,0,0", Rotation=Get Ambient Rot L1() -> (Ambient Rot L1:Rotator), Scale="1,1,1") {
    rotationMode=BMM_Additive
    translationMode=BMM_Additive
    translationSpace=BCS_BoneSpace
    alphaBoolBlend {blendOption=Linear}
    boneToModify {boneName=Dyn_Left___Float_Pieces1}
   }
   n21 = "Transform (Modify) Bone - Bone: Dyn_Left___Float_Pieces2"(ComponentPose=@n20, Alpha=1, Translation="0,0,0", Rotation=Get Ambient Rot L2() -> (Ambient Rot L2:Rotator), Scale="1,1,1") {
    rotationMode=BMM_Additive
    translationMode=BMM_Additive
    translationSpace=BCS_BoneSpace
    alphaBoolBlend {blendOption=Linear}
    boneToModify {boneName=Dyn_Left___Float_Pieces2}
   }
   n22 = "Transform (Modify) Bone - Bone: Dyn_Left___Float_Pieces3"(ComponentPose=@n21, Alpha=1, Translation="0,0,0", Rotation=Get Ambient Rot L3() -> (Ambient Rot L3:Rotator), Scale="1,1,1") {
    rotationMode=BMM_Additive
    translationMode=BMM_Additive
    translationSpace=BCS_BoneSpace
    alphaBoolBlend {blendOption=Linear}
    boneToModify {boneName=Dyn_Left___Float_Pieces3}
   }
   Output Pose(Result=Component To Local(ComponentPose=@n22)) {layerGroup=None}
   n23 = "Transform (Modify) Bone - Bone: Dyn_Right_Float_Pieces1"(Alpha=1, Translation="2,-1,0", Rotation="0,0,0", Scale="1,1,1") {
    rotationMode=BMM_Additive
    translationMode=BMM_Additive
    translationSpace=BCS_BoneSpace
    alphaBoolBlend {blendOption=Linear}
    boneToModify {boneName=Dyn_Right_Float_Pieces1}
    translation=<2 -1>
   }
   n24 = "Transform (Modify) Bone - Bone: Dyn_Left___Float_Pieces1"(ComponentPose=@n23, Alpha=1, Translation="1,1,0", Rotation="0,0,0", Scale="1,1,1") {
    rotationMode=BMM_Additive
    translationMode=BMM_Additive
    translationSpace=BCS_BoneSpace
    alphaBoolBlend {blendOption=Linear}
    boneToModify {boneName=Dyn_Left___Float_Pieces1}
    translation=<1 1>
   }
   n25 = "Transform (Modify) Bone - Bone: Dyn_Right_Float_Pieces2"(ComponentPose=@n24, Alpha=1, Translation="1,0,0", Rotation="0,0,0", Scale="1,1,1") {
    rotationMode=BMM_Additive
    translationMode=BMM_Additive
    translationSpace=BCS_BoneSpace
    alphaBoolBlend {blendOption=Linear}
    boneToModify {boneName=Dyn_Right_Float_Pieces2}
    translation {x=1}
   }
   n26 = "Transform (Modify) Bone - Bone: Dyn_Left___Float_Pieces2"(ComponentPose=@n25, Alpha=1, Translation="1,1,0", Rotation="0,0,0", Scale="1,1,1") {
    rotationMode=BMM_Additive
    translationMode=BMM_Additive
    translationSpace=BCS_BoneSpace
    alphaBoolBlend {blendOption=Linear}
    boneToModify {boneName=Dyn_Left___Float_Pieces2}
    translation=<1 1>
   }
   n27 = "Transform (Modify) Bone - Bone: Dyn_Right_Float_Pieces3"(ComponentPose=@n26, Alpha=1, Translation="1,1,-0.2", Rotation="0,0,0", Scale="1,1,1") {
    rotationMode=BMM_Additive
    translationMode=BMM_Additive
    translationSpace=BCS_BoneSpace
    alphaBoolBlend {blendOption=Linear}
    boneToModify {boneName=Dyn_Right_Float_Pieces3}
    translation=<1 1 -0.2>
   }
   n28 = "Transform (Modify) Bone - Bone: Dyn_Left___Float_Pieces3"(ComponentPose=@n27, Alpha=1, Translation="1,1,0.2", Rotation="0,0,0", Scale="1,1,1") {
    rotationMode=BMM_Additive
    translationMode=BMM_Additive
    translationSpace=BCS_BoneSpace
    alphaBoolBlend {blendOption=Linear}
    boneToModify {boneName=Dyn_Left___Float_Pieces3}
    translation=<1 1 0.2>
   }
   n29 = "Transform (Modify) Bone - Bone: Dyn_Right_Float_Pieces1"(Alpha=1, Translation="1.5,1,0.5", Rotation="0,0,0", Scale="1,1,1") {
    rotationMode=BMM_Additive
    translationMode=BMM_Additive
    translationSpace=BCS_BoneSpace
    alphaBoolBlend {blendOption=Linear}
    boneToModify {boneName=Dyn_Right_Float_Pieces1}
    translation=<1.5 1 0.5>
   }
   n30 = "Transform (Modify) Bone - Bone: Dyn_Left___Float_Pieces1"(ComponentPose=@n29, Alpha=1, Translation="0,-1,1", Rotation="0,0,0", Scale="1,1,1") {
    rotationMode=BMM_Additive
    translationMode=BMM_Additive
    translationSpace=BCS_BoneSpace
    alphaBoolBlend {blendOption=Linear}
    boneToModify {boneName=Dyn_Left___Float_Pieces1}
    translation {y=-1, z=1}
   }
   n31 = "Transform (Modify) Bone - Bone: Dyn_Right_Float_Pieces2"(ComponentPose=@n30, Alpha=1, Translation="2,2,0", Rotation="0,0,0", Scale="1,1,1") {
    rotationMode=BMM_Additive
    translationMode=BMM_Additive
    translationSpace=BCS_BoneSpace
    alphaBoolBlend {blendOption=Linear}
    boneToModify {boneName=Dyn_Right_Float_Pieces2}
    translation=<2 2>
   }
   n32 = "Transform (Modify) Bone - Bone: Dyn_Left___Float_Pieces2"(ComponentPose=@n31, Alpha=1, Translation="2,-2,0", Rotation="0,0,0", Scale="1,1,1") {
    rotationMode=BMM_Additive
    translationMode=BMM_Additive
    translationSpace=BCS_BoneSpace
    alphaBoolBlend {blendOption=Linear}
    boneToModify {boneName=Dyn_Left___Float_Pieces2}
    translation=<2 -2>
   }
   n33 = "Transform (Modify) Bone - Bone: Dyn_Right_Float_Pieces3"(ComponentPose=@n32, Alpha=1, Translation="2,-2,1", Rotation="0,0,0", Scale="1,1,1") {
    rotationMode=BMM_Additive
    translationMode=BMM_Additive
    translationSpace=BCS_BoneSpace
    alphaBoolBlend {blendOption=Linear}
    boneToModify {boneName=Dyn_Right_Float_Pieces3}
    translation=<2 -2 1>
   }
   n34 = "Transform (Modify) Bone - Bone: Dyn_Left___Float_Pieces3"(ComponentPose=@n33, Alpha=1, Translation="-1,0,0", Rotation="0,0,0", Scale="1,1,1") {
    rotationMode=BMM_Additive
    translationMode=BMM_Additive
    translationSpace=BCS_BoneSpace
    alphaBoolBlend {blendOption=Linear}
    boneToModify {boneName=Dyn_Left___Float_Pieces3}
    translation {x=-1}
   }
   n35 = "Transform (Modify) Bone - Bone: Dyn_Right_Float_Pieces1"(Alpha=1, Translation="1,0,-0.5", Rotation="0,0,0", Scale="1,1,1") {
    rotationMode=BMM_Additive
    translationMode=BMM_Additive
    translationSpace=BCS_BoneSpace
    alphaBoolBlend {blendOption=Linear}
    boneToModify {boneName=Dyn_Right_Float_Pieces1}
    translation {x=1, z=-0.5}
   }
   n36 = "Transform (Modify) Bone - Bone: Dyn_Left___Float_Pieces1"(ComponentPose=@n35, Alpha=1, Translation="1,0,0.5", Rotation="0,0,0", Scale="1,1,1") {
    rotationMode=BMM_Additive
    translationMode=BMM_Additive
    translationSpace=BCS_BoneSpace
    alphaBoolBlend {blendOption=Linear}
    boneToModify {boneName=Dyn_Left___Float_Pieces1}
    translation {x=1, z=0.5}
   }
   n37 = "Transform (Modify) Bone - Bone: Dyn_Right_Float_Pieces2"(ComponentPose=@n36, Alpha=1, Translation="2,0,0", Rotation="0,0,0", Scale="1,1,1") {
    rotationMode=BMM_Additive
    translationMode=BMM_Additive
    translationSpace=BCS_BoneSpace
    alphaBoolBlend {blendOption=Linear}
    boneToModify {boneName=Dyn_Right_Float_Pieces2}
    translation {x=2}
   }
   n38 = "Transform (Modify) Bone - Bone: Dyn_Left___Float_Pieces2"(ComponentPose=@n37, Alpha=1, Translation="2,0,0", Rotation="0,0,0", Scale="1,1,1") {
    rotationMode=BMM_Additive
    translationMode=BMM_Additive
    translationSpace=BCS_BoneSpace
    alphaBoolBlend {blendOption=Linear}
    boneToModify {boneName=Dyn_Left___Float_Pieces2}
    translation {x=2}
   }
   n39 = "Transform (Modify) Bone - Bone: Dyn_Right_Float_Pieces3"(ComponentPose=@n38, Alpha=1, Translation="2,0,1", Rotation="0,0,0", Scale="1,1,1") {
    rotationMode=BMM_Additive
    translationMode=BMM_Additive
    translationSpace=BCS_BoneSpace
    alphaBoolBlend {blendOption=Linear}
    boneToModify {boneName=Dyn_Right_Float_Pieces3}
    translation {x=2, z=1}
   }
   n40 = "Transform (Modify) Bone - Bone: Dyn_Left___Float_Pieces3"(ComponentPose=@n39, Alpha=1, Translation="2,0,0", Rotation="0,0,0", Scale="1,1,1") {
    rotationMode=BMM_Additive
    translationMode=BMM_Additive
    translationSpace=BCS_BoneSpace
    alphaBoolBlend {blendOption=Linear}
    boneToModify {boneName=Dyn_Left___Float_Pieces3}
    translation {x=2}
   }
   n41 = "Transform (Modify) Bone - Bone: Dyn_Right_Float_Pieces1"(ComponentPose=Local To Component(), Alpha=1, Translation="0,0,0", Rotation="0,0,0", Scale="1,1,1") {
    rotationMode=BMM_Additive
    translationMode=BMM_Additive
    translationSpace=BCS_BoneSpace
    alphaBoolBlend {blendOption=Linear}
    boneToModify {boneName=Dyn_Right_Float_Pieces1}
   }
   n42 = "Transform (Modify) Bone - Bone: Dyn_Left___Float_Pieces1"(ComponentPose=@n41, Alpha=1, Translation="0,0,0", Rotation="0,0,0", Scale="1,1,1") {
    rotationMode=BMM_Additive
    translationMode=BMM_Additive
    translationSpace=BCS_BoneSpace
    alphaBoolBlend {blendOption=Linear}
    boneToModify {boneName=Dyn_Left___Float_Pieces1}
   }
   n43 = "Transform (Modify) Bone - Bone: Dyn_Right_Float_Pieces2"(ComponentPose=@n42, Alpha=1, Translation="0,0,0", Rotation="0,0,0", Scale="1,1,1") {
    rotationMode=BMM_Additive
    translationMode=BMM_Additive
    translationSpace=BCS_BoneSpace
    alphaBoolBlend {blendOption=Linear}
    boneToModify {boneName=Dyn_Right_Float_Pieces2}
   }
   n44 = "Transform (Modify) Bone - Bone: Dyn_Left___Float_Pieces2"(ComponentPose=@n43, Alpha=1, Translation="0,0,0", Rotation="0,0,0", Scale="1,1,1") {
    rotationMode=BMM_Additive
    translationMode=BMM_Additive
    translationSpace=BCS_BoneSpace
    alphaBoolBlend {blendOption=Linear}
    boneToModify {boneName=Dyn_Left___Float_Pieces2}
   }
   n45 = "Transform (Modify) Bone - Bone: Dyn_Right_Float_Pieces3"(ComponentPose=@n44, Alpha=1, Translation="0,0,0", Rotation="0,0,0", Scale="1,1,1") {
    rotationMode=BMM_Additive
    translationMode=BMM_Additive
    translationSpace=BCS_BoneSpace
    alphaBoolBlend {blendOption=Linear}
    boneToModify {boneName=Dyn_Right_Float_Pieces3}
   }
   n46 = "Transform (Modify) Bone - Bone: Dyn_Left___Float_Pieces3"(ComponentPose=@n45, Alpha=1, Translation="0,0,0", Rotation="0,0,0", Scale="1,1,1") {
    rotationMode=BMM_Additive
    translationMode=BMM_Additive
    translationSpace=BCS_BoneSpace
    alphaBoolBlend {blendOption=Linear}
    boneToModify {boneName=Dyn_Left___Float_Pieces3}
   }
   n47 = Two Way Blend(A=Component To Local(ComponentPose=@n46), B=Component To Local(ComponentPose=@n40), Alpha=Get First Ring Up() -> (First Ring Up:double)) {blendNode {alphaBoolBlend {blendOption=Linear}, alphaScaleBiasClamp {bInterpResult=true, interpSpeedDecreasing=1, interpSpeedIncreasing=1}}}
   n48 = Two Way Blend(A=@n47, B=Component To Local(ComponentPose=@n34), Alpha=Get First Ring Middle() -> (First Ring Middle:double)) {blendNode {alphaBoolBlend {blendOption=Linear}, alphaScaleBiasClamp {bInterpResult=true, interpSpeedDecreasing=1, interpSpeedIncreasing=1}}}
   Two Way Blend(A=@n48, B=Component To Local(ComponentPose=@n28), Alpha=Get First Ring Down() -> (First Ring Down:double)) {blendNode {alphaBoolBlend {blendOption=Linear}, alphaScaleBiasClamp {bInterpResult=true, interpSpeedDecreasing=1, interpSpeedIncreasing=1}}}
  ]}
 EventGraph {nodes [
   Event BlueprintUpdateAnimation()
   > n1 = Sequence!(execute) -> (
    then_0
    then_1
    then_2:Exec
    then_3:Exec
    then_4:Exec
    then_5:Exec
    then_6:Exec
    then_7:Exec
    then_8:Exec
    then_9:Exec
    then_10:Exec
    then_11:Exec
   )
   > n2 = Ring_Animation!(
    execute
    self
    Time Offset 1:double=0
    Magnitude Multiplier:double=1
    Period:Vector="1,0.1,0.2"
    Magnitude:Vector="1,0.5,0.2"
    Period Multiplier:double=1
    Rotate Multiplier:double=1
   ) -> (then, VectorOut:Vector, RotateOut:Rotator)
   > Set Ambient_Trans_R1(Ambient_Trans_R1:Vector=@n2) -> (Output_Get:Vector)
   n3 = Ring_Animation(
    execute=@n1.then_1
    Time Offset 1=0
    Magnitude Multiplier=1
    Period="2,0.5,1"
    Magnitude="0.5,0.5,0.5"
    Period Multiplier=1
    Rotate Multiplier=1
   )
   > Set Ambient_Trans_R2(Ambient_Trans_R2:Vector=@n3) -> (Output_Get:Vector)
   n4 = Ring_Animation(execute=@n1.then_10, Time Offset 1=0, Magnitude Multiplier=1, Period="1, 0, 0", Magnitude="1,0,0", Period Multiplier=1, Rotate Multiplier=1)
   > Set Ambient Rot L2(Ambient Rot L2:Rotator=MakeRotator(Yaw=BreakRotator(InRot=@n4.RotateOut))) -> (Output_Get:Rotator)
   n5 = Ring_Animation(execute=@n1.then_11, Time Offset 1=0, Magnitude Multiplier=1, Period="1, 0, 0", Magnitude="1,0,0", Period Multiplier=1, Rotate Multiplier=1)
   > Set Ambient Rot L3(Ambient Rot L3:Rotator=MakeRotator(Yaw=BreakRotator(InRot=@n5.RotateOut))) -> (Output_Get:Rotator)
   n6 = Ring_Animation(
    execute=@n1.then_2
    Time Offset 1=0
    Magnitude Multiplier=1
    Period="1,1,0.3"
    Magnitude="0.8,0.2,0.5"
    Period Multiplier=1
    Rotate Multiplier=1
   )
   > Set Ambient_Trans_R3(Ambient_Trans_R3:Vector=@n6) -> (Output_Get:Vector)
   n7 = Ring_Animation(
    execute=@n1.then_3
    Time Offset 1=0
    Magnitude Multiplier=1
    Period="0.8,0.2,0.3"
    Magnitude="1,0.3,0.5"
    Period Multiplier=1
    Rotate Multiplier=1
   )
   > Set Ambient_Trans_L1(Ambient_Trans_L1:Vector=@n7) -> (Output_Get:Vector)
   n8 = Ring_Animation(
    execute=@n1.then_4
    Time Offset 1=0
    Magnitude Multiplier=1
    Period="1.5,0.5,1"
    Magnitude="0.5,0.5,0.8"
    Period Multiplier=1
    Rotate Multiplier=1
   )
   > Set Ambient_Trans_L2(Ambient_Trans_L2:Vector=@n8) -> (Output_Get:Vector)
   n9 = Ring_Animation(
    execute=@n1.then_5
    Time Offset 1=0
    Magnitude Multiplier=1
    Period="1,1,0.3"
    Magnitude="1,0.2,0.5"
    Period Multiplier=1
    Rotate Multiplier=1
   )
   > Set Ambient_Trans_L3(Ambient_Trans_L3:Vector=@n9) -> (Output_Get:Vector)
   n10 = Ring_Animation(execute=@n1.then_6, Time Offset 1=0, Magnitude Multiplier=1, Period="1, 0, 0", Magnitude="1,0,0", Period Multiplier=1, Rotate Multiplier=1)
   > Set Ambient Rot R1(Ambient Rot R1:Rotator=MakeRotator(Yaw=BreakRotator(InRot=@n10.RotateOut))) -> (Output_Get:Rotator)
   n11 = Ring_Animation(execute=@n1.then_7, Time Offset 1=0, Magnitude Multiplier=1, Period="1, 0, 0", Magnitude="1,0,0", Period Multiplier=1, Rotate Multiplier=1)
   > Set Ambient Rot R2(Ambient Rot R2:Rotator=MakeRotator(Yaw=BreakRotator(InRot=@n11.RotateOut))) -> (Output_Get:Rotator)
   n12 = Ring_Animation(execute=@n1.then_8, Time Offset 1=0, Magnitude Multiplier=1, Period="1, 0, 0", Magnitude="1,0,0", Period Multiplier=1, Rotate Multiplier=1)
   > Set Ambient Rot R3(Ambient Rot R3:Rotator=MakeRotator(Yaw=BreakRotator(InRot=@n12.RotateOut))) -> (Output_Get:Rotator)
   n13 = Ring_Animation(execute=@n1.then_9, Time Offset 1=0, Magnitude Multiplier=1, Period="1, 0, 0", Magnitude="1,0,0", Period Multiplier=1, Rotate Multiplier=1)
   > Set Ambient Rot L1(Ambient Rot L1:Rotator=MakeRotator(Yaw=BreakRotator(InRot=@n13.RotateOut))) -> (Output_Get:Rotator)
   TryGetPawnOwner()
   RetriggerableDelay(Duration=2)
   RetriggerableDelay(Duration=0.1)
   RetriggerableDelay(Duration=2)
   Delay(execute=@n15.then, Duration=1)
   > Set First Ring Middle(First Ring Middle:double=1) -> (Output_Get:double)
   > Delay(Duration=1)
   > Set First Ring Down(First Ring Down:double=1) -> (Output_Get:double)
   > Delay(Duration=1)
   > Set First Ring Up(First Ring Up:double=0) -> (Output_Get:double)
   > Delay(Duration=0.1)
   > Set First Ring Middle(First Ring Middle=0)
   > Delay(Duration=0.1)
   > n14 = Set First Ring Down(First Ring Down=0)
   Do Once(Reset=@n14.then)
   > n15 = Set First Ring Up(First Ring Up=1)
  ]}
 Ring_Animation {nodes [
   n1 = Ring_Animation!() -> (
    then
    Time Offset 1:double=~
    Magnitude Multiplier:double=~
    Period:Vector=~
    Magnitude:Vector="10,0,0"
    Period Multiplier:double=1
    Rotate Multiplier:double=~
   ) {is=title K2Node_FunctionEntry}
   n2 = BreakVector(InVec=@n1.Magnitude)
   n3 = BreakVector(InVec=@n1.Period)
   n4 = GetGameTimeInSeconds()
   n5 = float * float(A=float * float(A=@n1.Period Multiplier, B=@n3.Y), B=@n4)
   n6 = float * float(A="Sin (Radians)"(A=float + float(A=@n5, B=@n1) {is=title K2Node_CommutativeAssociativeBinaryOperator}), B=@n2.Y)
   n7 = float * float(A=float * float(A=@n1.Period Multiplier, B=@n3.Z), B=@n4)
   n8 = float * float(A="Sin (Radians)"(A=float + float(A=@n7, B=@n1)), B=@n2.Z)
   n9 = float * float(A=float * float(A=@n1.Period Multiplier, B=@n3), B=@n4)
   n10 = float * float(A="Sin (Radians)"(A=float + float(A=@n9, B=@n1)), B=@n2)
   n11 = MakeRotator(Roll=@n10, Pitch=@n6, Yaw=@n8)
   > Return Node!(
    execute
    VectorOut:Vector=vector * float(A=MakeVector(X=@n10, Y=@n6, Z=@n8), B=@n1.Magnitude Multiplier)
    RotateOut:Rotator="Scale Rotator (Float)"!(A:Rotator="Scale Rotator (Float)"(A=@n11, B=@n1.Magnitude Multiplier), B:float=@n1.Rotate Multiplier) -> (ReturnValue:Rotator)
   )
  ]}
]
variables [
 Ambient Rot L1 {type=Rotator}
 Ambient Rot L2 {type=Rotator}
 Ambient Rot L3 {type=Rotator}
 Ambient Rot R1 {type=Rotator}
 Ambient Rot R2 {type=Rotator}
 Ambient Rot R3 {type=Rotator}
 Ambient_Trans_L1 {type=Vector}
 Ambient_Trans_L2 {type=Vector}
 Ambient_Trans_L3 {type=Vector}
 Ambient_Trans_R1 {type=Vector}
 Ambient_Trans_R2 {type=Vector}
 Ambient_Trans_R3 {type=Vector}
 First Ring Down {type=number}
 First Ring Middle {type=number}
 First Ring Up {type=number}
 Floating Down State {type=number}
 Floating Up State {type=number}
 Restart {type=number}
 Start State {type=number}
]
